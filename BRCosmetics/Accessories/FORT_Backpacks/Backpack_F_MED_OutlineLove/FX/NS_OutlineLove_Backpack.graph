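# Niagara system for the OutlineLove backpack cosmetic, placing flickering flames, glows and rising embers at eight fixed points.
# FlamesAndGlows burst-spawns sprite, mesh and light particles, animating size and color with waveforms and curves, with a front-end brightness modifier.
# Its lagged-position delta feeds dynamic material parameters so the material can react to movement.
# Embers bursts sprites from the same offsets, driven by upward acceleration, drag and curl noise, cooling in color while fading.
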NiagaraSystem /BRCosmetics/Accessories/FORT_Backpacks/Backpack_F_MED_OutlineLove/FX/NS_OutlineLove_Backpack
dependencies {
 usedDataInterfaces [&NiagaraDataInterfaceCurve, &NiagaraDataInterfaceColorCurve]
 usedDynamicInputs []
 usedModules [&SystemState]
}
emitters {
 Embers {
  emitterUpdateScript [
   &EmitterState(
    Enable Distance Culling=true
    Enable Visibility Culling=true
    MaxDistance=3000
    Module.bUseMaxDistance=true
    Module.bUseMinDistance=false
    Scalability Mode=::NewEnumerator1
   )
   &SpawnBurst_Instantaneous(Spawn Count=35, Use Spawn Probability=false, UseLoopCountLimit=false)
  ]
  particleSpawnScript [
   &FN_SetSpawnGroup(Number of Spawn Groups=8)
   /Niagara/Modules/Spawn/Initialization/InitializeParticle(
    Color=<rgba 1 0.3667 0 0.5>
    Lifetime=&UniformRangedFloat(Minimum=1, Maximum=2)
    Module.Write Color=true
    Module.Write Lifetime=true
    Module.Write Mass=true
    Module.Write Material Random=true
    Module.Write Position=true
    Module.Write Scale=false
    Module.Write SpriteRotation=false
    Module.Write SpriteSize=true
    Position=&AddVectorToPosition(
     Position=/Niagara/DynamicInputs/Helpers/SimulationPosition()
     Vector=&MergeHelper_VectorSwitch(
      0=<13.2843 -4.8409 -4.9112>
      1=<4.2918 -5.2953 -4.9112>
      2=<-5.5481 -4.8938 -4.9112>
      3=<-13.4392 -4.8301 -5.0491>
      4=<13.5294 -16.2061 -4.9266>
      5=<5.2633 -16.049 -4.8448>
      6=<-4.4348 -15.6865 -4.8448>
      7=<-13.5253 -16.1743 -4.8448>
      Index=$Particles.SpawnGroup
      Num=::NewEnumerator6
     )
    )
    Sprite Size=&Vector2DFromFloat(Value=&FloatFromCurve(
      FloatCurve=DataInterface() {propertyValues {Curve {keys [<at 0 0.8>,<at 0.15 0.75>,<at 0.2996 0.5>,<at 1 0.3>]}}}
      CurveIndex=&UniformRangedFloat(Minimum=0, Maximum=1)
      Scale Curve=0.9
     ))
   )
   &AddVelocity(Velocity=<0 0 1>, Velocity Speed Scale=&RandomRangeFloat(Minimum=10, Maximum=25))
   &ShapeLocation(MaskSpawnGroup=false, Override Seed=false, Sphere Radius=2.5, UseOwnerScale=true)
  ]
  particleUpdateScript [
   &ParticleState()
   &AccelerationForce(Acceleration=&UniformRangedVector(Minimum=<-1 0 10>, Maximum=<-2 0 50>))
   &Drag(
    Drag=&FloatFromCurve(
     FloatCurve=DataInterface() {propertyValues {Curve {keys [<at 0 0.5>,<at 0.5 0.5>,<at 1 1>]}}}
     Scale Curve=&RandomRangeFloat(Minimum=5, Maximum=10)
    )
    Use Linear Drag=true
    Use Rotational Drag=false
   )
   /Niagara/Modules/Update/Forces/CurlNoiseForce(
    Noise Frequency=0.04
    Noise Strength=&FloatFromCurve(FloatCurve=DataInterface() {propertyValues {Curve {keys [<at 0 1>,<at 1 0>]}}}, Scale Curve=25)
    Pan Noise=true
    Pan Noise Field=<0 0 1>
   )
   &SolveForcesAndVelocity(Clamp Velocity=false, Limit Acceleration=false)
   &ScaleSpriteSizeBySpeed(
    Max Scale Factor=<2 5>
    Sample Scale Factor By Curve=true
    Scale Factor Curve=DataInterface() {propertyValues {Curve {keys [
        <at -0.0077 0.0765 RCIM_Cubic>
        {arriveTangent=1.2463, interpMode=RCIM_Cubic, leaveTangent=1.2463, tangentMode=RCTM_User, time=0.9985, value=0.9549}
       ]}}}
    Velocity Threshold=500
   )
   &ScaleColor(
    Scale Alpha=&FloatFromCurve(FloatCurve=DataInterface() {propertyValues {Curve {keys [<at 0 1>,<at 1 0>]}}})
    Scale RGB=&MakeVectorFromLinearColorRGB(Color=&ColorFromCurve(ColorCurve=DataInterface() {propertyValues {
        AlphaCurve {keys [<at 0 1 RCIM_Cubic>,<at 0.7107 1>]}
        BlueCurve {keys [<at 0 50 RCIM_Cubic>,<at 0.2499 1.5104>,<at 0.7114 0.1753 RCIM_Cubic>]}
        GreenCurve {keys [<at 0 32.3351 RCIM_Cubic>,<at 0.2499 3.7743>,<at 0.7114 0.409 RCIM_Cubic>]}
        RedCurve {keys [<at 0 14.6701 RCIM_Cubic>,<at 0.2499 10>,<at 0.7114 1 RCIM_Cubic>]}
       }}))
    ScaleA=true
    ScaleRGB=true
   )
  ]
  rendererClasses [&NiagaraSpriteRendererProperties]
  renderers [{rendererClass=&NiagaraSpriteRendererProperties}]
 }
 FlamesAndGlows {
  emitterUpdateScript [
   &EmitterState()
   &SpawnBurst_Instantaneous(Spawn Count=8, Use Spawn Probability=false, UseLoopCountLimit=false)
   &SpawnBurst_Instantaneous(Spawn Group=1, Use Spawn Probability=false, UseLoopCountLimit=false) {module=SpawnBurst_Instantaneous001}
  ]
  particleSpawnScript [
   /SetVariables(Particles.VisibilityTag=&SpawnGroup()) {isSetParameters=true}
   &PartitionParticles(Number Of Partitions=8)
   /Niagara/Modules/Spawn/Initialization/V2/InitializeParticle(
    Color=<rgba 8 0.8133 0.16 3>
    Lifetime Max=2.8
    Lifetime Min=2.2
    Lifetime Mode=::NewEnumerator1
    Mesh Scale=&MakeVector(X=&RandomRangeFloat(Minimum=1.7, Maximum=1.95), Y=2, Z=2)
    Mesh Scale Mode=::NewEnumerator4
    Position Offset=&MergeHelper_VectorSwitch(
     0=<13.2843 -4.8409 -4.9112>
     1=<4.2918 -5.2953 -4.9112>
     2=<-5.5481 -4.8938 -4.9112>
     3=<-13.4392 -4.8301 -5.0491>
     4=<13.5294 -16.2061 -4.9266>
     5=<5.2633 -16.049 -4.8448>
     6=<-4.4348 -15.6865 -4.8448>
     7=<-13.5253 -16.1743 -4.8448>
     Index=$Particles.PartitionParticles.Partition
     Num=::NewEnumerator6
    )
    Sprite Size Mode=::NewEnumerator1
    Uniform Sprite Size Max=12
    Uniform Sprite Size Min=10
    UsePositionOffset=true
    Write Material Random=true
   )
   /SetVariables(Particles.ColorAlt:LinearColor=<rgba 1 0.3667 0 0.5>) {isSetParameters=true}
  ]
  particleUpdateScript [
   &ParticleState()
   &ScaleMeshSize(Scale Factor=&Multiply_VectorByFloat(
     Vector=&MakeVector(
      X=&Waveform(
       Global Input Position=&RandomRangeFloat(Minimum=0, Maximum=1)
       "Waveform [1]"=::NewEnumerator0
       "[1] Frequency"=1
       "[1] Amplitude Min/Max"=<0.7 1>
      )
      Y=&FloatFromCurve(FloatCurve=DataInterface() {propertyValues {Curve {
          defaultValue=0
          keys [
           {arriveTangent=-2.1363, interpMode=RCIM_Cubic, leaveTangent=-2.1363, tangentMode=RCTM_User, time=0, value=1}
           {arriveTangent=-0.0696, interpMode=RCIM_Cubic, leaveTangent=-0.0696, tangentMode=RCTM_User, time=0.5, value=0.5}
           {arriveTangent=-0.5552, interpMode=RCIM_Cubic, leaveTangent=-0.5552, tangentMode=RCTM_User, time=1, value=0}
          ]
         }}})
      Z=&FloatFromCurve(FloatCurve=DataInterface() {propertyValues {Curve {
          defaultValue=0
          keys [
           {arriveTangent=-2.1363, interpMode=RCIM_Cubic, leaveTangent=-2.1363, tangentMode=RCTM_User, time=0, value=1}
           {arriveTangent=-0.0696, interpMode=RCIM_Cubic, leaveTangent=-0.0696, tangentMode=RCTM_User, time=0.5, value=0.5}
           {arriveTangent=-0.5552, interpMode=RCIM_Cubic, leaveTangent=-0.5552, tangentMode=RCTM_User, time=1, value=0}
          ]
         }}})
     )
     Float=&FloatFromCurve(FloatCurve=DataInterface() {propertyValues {Curve {
         defaultValue=0
         keys [
          {interpMode=RCIM_Cubic, tangentMode=RCTM_User, time=0, value=0}
          {interpMode=RCIM_Cubic, tangentMode=RCTM_User, time=0.1547, value=1}
          {arriveTangent=-1.2196, interpMode=RCIM_Cubic, leaveTangent=-1.2196, tangentMode=RCTM_User, time=0.2109, value=0.65}
          {arriveTangent=-0.2193, interpMode=RCIM_Cubic, leaveTangent=-0.2193, time=0.5, value=0.5}
          {arriveTangent=-1, interpMode=RCIM_Cubic, leaveTangent=-1, time=0.8948, value=0.5}
          {arriveTangent=-1.8247, tangentMode=RCTM_User, time=1, value=0}
         ]
        }}})
    ))
   &SolveForcesAndVelocity(Clamp Velocity=false, Limit Acceleration=false)
   &ScaleColor(
    Scale Alpha=&Multiply_Float(
     A=&FloatFromCurve(FloatCurve=DataInterface() {propertyValues {Curve {
         defaultValue=1
         keys [
          {arriveTangent=9.0725, interpMode=RCIM_Cubic, leaveTangent=9.0725, tangentMode=RCTM_User, time=0, value=0}
          {arriveTangent=-4.7692, interpMode=RCIM_Cubic, leaveTangent=-4.7692, tangentMode=RCTM_User, time=0.0704, value=0.9847}
          {arriveTangent=-0.2605, interpMode=RCIM_Cubic, leaveTangent=-0.2605, tangentMode=RCTM_User, time=0.2926, value=0.4}
          {arriveTangent=-0.5654, interpMode=RCIM_Cubic, leaveTangent=-0.5654, time=0.8801, value=0.2906}
          {arriveTangent=-1.7718, interpMode=RCIM_Cubic, leaveTangent=-1.7718, tangentMode=RCTM_User, time=1, value=0}
         ]
        }}})
     B=&Waveform(
      Global Input Position=&RandomRangeFloat(Minimum=0, Maximum=1)
      Global Phase=$System.NormalizedLoopAge
      "Waveform [1]"=::NewEnumerator12
      "[1] Frequency"=1
      "[1] Amplitude Min/Max"=<0.5 1>
     )
    )
    Scale RGB=&VectorFromFloat(Value=&Waveform(
      Global Input Position=&RandomRangeFloat(Minimum=0, Maximum=1)
      Global Phase=$Particles.NormalizedAge
      "Waveform [1]"=::NewEnumerator12
      "[1] Frequency"=2
      "[1] Amplitude Min/Max"=<1 2>
     ))
    ScaleA=true
    ScaleRGB=true
   )
   &ScaleSpriteSize(Scale Factor=&Multiply_Vector2D_ByFloat(
     Float=&Waveform("Waveform [1]"=::NewEnumerator12, "[1] Frequency"=5, "[1] Amplitude Min/Max"=<1 2>, "[1] Scale"=0.5, "[1] Bias"=0.5)
     Vector2D=<1 1>
    ))
   /SetVariables(
    Particles.ColorAlt:LinearColor=&MultiplyLinearColor(
     Linear Color=$Particles.Initial.ColorAlt
     Scale Factor=&MakeVector4_FromFloat(Float=&Waveform(
       Global Amplitude Scale=&FloatFromCurve(FloatCurve=DataInterface() {propertyValues {Curve {
           defaultValue=1
           keys [
            {arriveTangent=9.0725, interpMode=RCIM_Cubic, leaveTangent=9.0725, tangentMode=RCTM_User, time=0, value=0}
            {arriveTangent=-4.7692, interpMode=RCIM_Cubic, leaveTangent=-4.7692, tangentMode=RCTM_User, time=0.0704, value=0.9847}
            {arriveTangent=-0.2605, interpMode=RCIM_Cubic, leaveTangent=-0.2605, tangentMode=RCTM_User, time=0.2926, value=0.75}
            {arriveTangent=-0.2068, interpMode=RCIM_Cubic, leaveTangent=-0.2068, tangentMode=RCTM_User, time=0.9242, value=0.75}
            {arriveTangent=-32.0623, interpMode=RCIM_Cubic, leaveTangent=-32.0623, tangentMode=RCTM_User, time=1, value=0}
           ]
          }}})
       "Waveform [1]"=::NewEnumerator12
       "[1] Input Position"=&ReturnNormalizedExecIndex(Normalized Index Scale=0.8)
       "[1] Frequency"=3
       "[1] Amplitude Min/Max"=<0.5 1>
      ))
    )
    Particles.SpriteSizeAlt:Vector2f=&Vector2DFromFloat(Value=&FloatFromCurve(
      FloatCurve=DataInterface() {propertyValues {Curve {keys [
          {arriveTangent=-0.4528, interpMode=RCIM_Cubic, leaveTangent=-0.4528, tangentMode=RCTM_User, time=0, value=1}
          {arriveTangent=-1, interpMode=RCIM_Cubic, leaveTangent=-1, time=0.8796, value=0.7676}
          {arriveTangent=-15.2452, interpMode=RCIM_Cubic, leaveTangent=-15.2452, tangentMode=RCTM_User, time=1, value=0}
         ]}}}
      Scale Curve=&RandomRangeFloat(Minimum=5, Maximum=10)
     ))
   ) {isSetParameters=true}
   /SetVariables(Particles.MaterialRandom:NiagaraFloat=&RandomRangeFloat(Minimum=0, Maximum=1)) {isSetParameters=true}
   /SetVariables(Particles.LaggedPosition:Vector3f=/Niagara/DynamicInputs/Lerp/SmoothLerpOverTime_Position(
     Convergence Rate:NiagaraFloat=18
     Smooth Value:NiagaraPosition=Linked() {linkedVariable {name=Particles.LaggedPosition, type=Vector3f}}
     Target Value:NiagaraPosition=$Particles.Position
    )) {isSetParameters=true}
   /SetVariables(Particles.PositionDelta:Vector3f=&Subtract_Vector(A=$Particles.LaggedPosition, B=Linked() {linkedVariable {name=Particles.Position, type=NiagaraPosition}})) {isSetParameters=true}
   &ScaleColor(
    Scale RGB=&VectorFromFloat(Value=&FN_FrontEndFloatModifier(Gameplay=1, Front End=0.15))
    ScaleA=false
    ScaleRGB=true
   ) {module=ScaleColor001}
   &DynamicMaterialParameters(
    Index 0 Param 1=&Lerp_Float(
     A=1
     B=2
     Alpha=&Waveform(
      Global Input Position=&RandomRangeFloat(Minimum=0, Maximum=1)
      "Waveform [1]"=::NewEnumerator12
      "[1] Frequency"=5
      "[1] Amplitude Min/Max"=<0 1>
     )
    )
    Index 0 Param 2=&MakeFloatFromVector(Vector=$Particles.PositionDelta, Channel=::NewEnumerator0)
    Index 0 Param 3=&MakeFloatFromVector(Vector=$Particles.PositionDelta, Channel=::NewEnumerator1)
    Index 0 Param 4=&ClampFloat(Float=&MakeFloatFromVector(Vector=$Particles.PositionDelta, Channel=::NewEnumerator2), Max=2)
    Module.Index0Param0Write=true
    Module.Index0Param1Write=true
    Module.Index0Param2Write=true
    Module.Index0Param3Write=true
   )
   &CameraOffset(Camera Offset Amount=2)
   &FN_Light_Attributes(
    Color=<rgba 10 1.0167 0.2 3>
    Create Color Attribute=true
    Create Exponent Attribute=true
    Create Position Attribute=true
    Create Radius Attribute=true
    Exponent=1
    Position=<0 -15 0>
    Radius=20
    Scale Color=&FloatFromCurve(FloatCurve=DataInterface() {propertyValues {Curve {
        defaultValue=1
        keys [
         {arriveTangent=9.0725, interpMode=RCIM_Cubic, leaveTangent=9.0725, tangentMode=RCTM_User, time=0, value=0}
         {arriveTangent=-2.5249, interpMode=RCIM_Cubic, leaveTangent=-2.5249, tangentMode=RCTM_User, time=0.1408, value=0.9847}
         {arriveTangent=-0.0405, interpMode=RCIM_Cubic, leaveTangent=-0.0405, tangentMode=RCTM_User, time=1, value=0}
        ]
       }}})
   )
  ]
  rendererClasses [
   &NiagaraSpriteRendererProperties
   &NiagaraMeshRendererProperties
   &NiagaraLightRendererProperties
  ]
  renderers [
   {rendererClass=&NiagaraSpriteRendererProperties}
   {rendererClass=&NiagaraMeshRendererProperties, rendererIndex=1}
   {rendererClass=&NiagaraLightRendererProperties, rendererIndex=2}
  ]
 }
}
summary {
 systemName=NS_OutlineLove_Backpack
 emitters [
  {
   emitterName=FlamesAndGlows
   rendererClasses [
    &NiagaraSpriteRendererProperties
    &NiagaraMeshRendererProperties
    &NiagaraLightRendererProperties
   ]
  }
  {emitterName=Embers, rendererClasses [&NiagaraSpriteRendererProperties]}
 ]
 userVariables [
  User.bIsFrontEnd {type=NiagaraBool}
  User.bIsFrontEndPreview {type=NiagaraBool}
 ]
}
