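# Physics asset for the OutlineLove backpack cosmetic, giving three egg pieces on the pack constrained secondary motion.
# The pack body uses two parallel capsules, each base_egg bone has a tiny capsule, and each dyn_egg bone has a larger one.
# Each dyn_egg links to its base_egg with tight swing and twist limits, while each base_egg links to the pack with free swing and twist.
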PhysicsAsset /BRCosmetics/Accessories/FORT_Backpacks/Backpack_F_MED_OutlineLove/Meshes/F_MED_OutlineLove_Pack_Physics_FE
bodies {
 base_egg_1 [{shapeType=Capsule, length=1, radius=0.5, shapeName=base_egg_3_capsule}]
 base_egg_2 [{shapeType=Capsule, length=1, radius=0.5, shapeName=base_egg_3_capsule}]
 base_egg_3 [{shapeType=Capsule, length=1, radius=0.5, shapeName=base_egg_3_capsule}]
 dyn_egg_1 [{shapeType=Capsule, length=8.7281, radius=1.9921, shapeName=dyn_egg_3_capsule, rotation=<rot 90 0 0>}]
 dyn_egg_2 [{shapeType=Capsule, length=8.7281, radius=1.9921, shapeName=dyn_egg_3_capsule, rotation=<rot 90 0 0>}]
 dyn_egg_3 [{shapeType=Capsule, length=8.7281, radius=1.9921, shapeName=dyn_egg_3_capsule, rotation=<rot 90 0 0>}]
 pack [
  {shapeType=Capsule, length=20.747, radius=3.5136, shapeName=pack_capsule, center=<0.0836 -4.7666 -12.2074>, rotation=<rot 0 -90.865 -89.962>}
  {
   shapeType=Capsule
   length=20.747
   radius=3.5136
   shapeName=pack_capsule
   center=<-0.0937 -16.5119 -12.2074>
   rotation=<rot 0 -90.865 -89.962>
  }
 ]
}
constraints [
 {bone1Name=dyn_egg_1, bone2Name=base_egg_1, swing1LimitDegrees=15, swing2LimitDegrees=20, twistLimitDegrees=10}
 {bone1Name=dyn_egg_2, bone2Name=base_egg_2, swing1LimitDegrees=15, swing2LimitDegrees=20, twistLimitDegrees=10}
 {bone1Name=dyn_egg_3, bone2Name=base_egg_3, swing1LimitDegrees=15, swing2LimitDegrees=20, twistLimitDegrees=10}
 {
  bone1Name=base_egg_1
  bone2Name=pack
  swing1LimitDegrees=45
  swing1Motion=Free
  swing2LimitDegrees=45
  swing2Motion=Free
  twistLimitDegrees=45
  twistMotion=Free
 }
 {
  bone1Name=base_egg_2
  bone2Name=pack
  swing1LimitDegrees=45
  swing1Motion=Free
  swing2LimitDegrees=45
  swing2Motion=Free
  twistLimitDegrees=45
  twistMotion=Free
 }
 {
  bone1Name=base_egg_3
  bone2Name=pack
  swing1LimitDegrees=45
  swing1Motion=Free
  swing2LimitDegrees=45
  swing2Motion=Free
  twistLimitDegrees=45
  twistMotion=Free
 }
]
