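# Anim Blueprint for the OutlineLove backpack that offsets the pack and adds lid motion and secondary physics.
# Additive bone transforms on pack and base_lid feed a cached pose, then two differently tuned AnimDynamics (lid, lock) plus RigidBody chains, selected by bIsInFrontEnd.
# The event graph derives lid-close and pack offset values from spine bend, a gender-based offset, movement-smoothed gravity, and a one-second flag on enemy eliminations.
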AnimBlueprint /BRCosmetics/Accessories/FORT_Backpacks/Backpack_F_MED_OutlineLove/Meshes/F_MED_OutlineLove_Pack_AnimBP
target_skeleton=&/F_MED_OutlineLove_Pack_Skeleton
graphs [
 AnimGraph {nodes [
   n1 = Anim Dynamics - Bone: dyn_lid!(ComponentPose=Local To Component(LocalPose=Use cached pose 'preDynamicsCP'()), Alpha=1) -> (Pose) {
    angularDampingOverride=1
    angularSpringConstant=12000
    bAngularSpring=true
    bOverrideAngularDamping=true
    bUseGravityOverride=true
    gravityScale=0.35
    simulationSpace=RootRelative
    boundBone {boneName=dyn_lid}
    componentAppliedLinearAccClamp=<4000 4000 4000>
    componentLinearAccScale=<1 1 0.8>
    componentLinearVelScale=<1 1 0.8>
    gravityOverride {z=-650}
    physicsBodyDefinitions [{
      boundBone {boneName=dyn_lid}
      boxExtents=<20 30 30>
      constraintSetup {angularLimitsMax=<0 14 0>, angularLimitsMin=<0 -15 0>, angularTarget=<1 0 0>}
      localJointOffset=<4 0 0>
     }]
    simSpaceSettings {maxAngularAcceleration=2, maxAngularVelocity=6, simSpaceAngularAlpha=1}
   }
   n2 = Anim Dynamics - Bone: dyn_lock!(ComponentPose=@n1, Alpha=1) -> (Pose) {
    angularDampingOverride=1
    angularSpringConstant=12000
    bAngularSpring=true
    bOverrideAngularDamping=true
    bUseGravityOverride=true
    gravityScale=0.35
    simulationSpace=RootRelative
    boundBone {boneName=dyn_lock}
    componentAppliedLinearAccClamp=<4000 4000 4000>
    componentLinearAccScale=<1 1 0.8>
    componentLinearVelScale=<1 1 0.8>
    gravityOverride {z=-650}
    physicsBodyDefinitions [{
      boundBone {boneName=dyn_lock}
      boxExtents=<20 20 20>
      constraintSetup {angularLimitsMax=<0 4 0>, angularLimitsMin=<0 -70 0>, angularTarget=<1 0 0>}
      localJointOffset=<4 0 0>
     }]
    simSpaceSettings {maxAngularAcceleration=1, maxAngularVelocity=6, simSpaceAngularAlpha=1}
   }
   n3 = RigidBody!(ComponentPose=@n2, Alpha, OverrideWorldGravity:Vector="0,0,-200") -> (Pose) {
    bOverrideWorldGravity=true
    overridePhysicsAsset=&/F_MED_OutlineLove_Pack_Physics_FE
    simulationSpace=BaseBoneSpace
    alphaBoolBlend {blendOption=Linear}
    baseBoneRef {boneName=pack}
    componentAppliedLinearAccClamp=<4000 4000 4000>
    overrideWorldGravity {z=-200}
    simSpaceSettings {maxAngularAcceleration=2, maxAngularVelocity=6, maxLinearAcceleration=120, maxLinearVelocity=50, worldAlpha=1, externalLinearDragV=<2 2 2>}
   }
   n4 = Anim Dynamics - Bone: dyn_lid(ComponentPose=Local To Component(LocalPose=Use cached pose 'preDynamicsCP'()), Alpha=1) {
    angularDampingOverride=1
    angularSpringConstant=16000
    bAngularSpring=true
    bOverrideAngularDamping=true
    bUseGravityOverride=true
    gravityScale=0.35
    simulationSpace=RootRelative
    boundBone {boneName=dyn_lid}
    componentAppliedLinearAccClamp=<4000 4000 4000>
    componentLinearAccScale=<1 1 0.8>
    componentLinearVelScale=<1 1 0.8>
    gravityOverride {z=-980}
    physicsBodyDefinitions [{
      boundBone {boneName=dyn_lid}
      boxExtents=<20 30 30>
      constraintSetup {angularLimitsMax=<0 14 0>, angularLimitsMin=<0 -15 0>, angularTarget=<1 0 0>}
      localJointOffset=<6 0 0>
     }]
    simSpaceSettings {maxAngularAcceleration=1, maxAngularVelocity=6, simSpaceAngularAlpha=1}
   }
   n5 = Anim Dynamics - Bone: dyn_lock(ComponentPose=@n4, Alpha=1) {
    angularDampingOverride=1
    angularSpringConstant=16000
    bAngularSpring=true
    bOverrideAngularDamping=true
    bUseGravityOverride=true
    gravityScale=0.35
    simulationSpace=RootRelative
    boundBone {boneName=dyn_lock}
    componentAppliedLinearAccClamp=<4000 4000 4000>
    componentLinearAccScale=<1 1 0.8>
    componentLinearVelScale=<1 1 0.8>
    gravityOverride {z=-980}
    physicsBodyDefinitions [{
      boundBone {boneName=dyn_lock}
      boxExtents=<20 20 20>
      constraintSetup {angularLimitsMax=<0 4 0>, angularLimitsMin=<0 -70 0>, angularTarget=<1 0 0>}
      localJointOffset=<6 0 0>
     }]
    simSpaceSettings {maxAngularAcceleration=1, maxAngularVelocity=6, simSpaceAngularAlpha=1}
   }
   n6 = RigidBody(ComponentPose=@n5, OverrideWorldGravity="0,0,-200") {
    bOverrideWorldGravity=true
    overridePhysicsAsset=&/F_MED_OutlineLove_Pack_Physics
    simulationSpace=BaseBoneSpace
    alphaBoolBlend {blendOption=Linear}
    baseBoneRef {boneName=pack}
    componentAppliedLinearAccClamp=<500 500 500>
    overrideWorldGravity {z=-200}
    simSpaceSettings {maxAngularAcceleration=2, maxAngularVelocity=6, maxLinearAcceleration=120, maxLinearVelocity=50, worldAlpha=1}
   }
   Output Pose(Result=Blend Poses by bool(BlendPose_0=Component To Local(ComponentPose=@n3), BlendPose_1=Component To Local(ComponentPose=@n6), bActiveValue=Get bIsInFrontEnd())) {layerGroup=None}
   n7 = Input Pose()
   n8 = "Transform (Modify) Bone - Bone: pack"!(
    ComponentPose="Transform (Modify) Bone - Bone: pack"!(ComponentPose=Local To Component(LocalPose=@n7), Alpha, Translation="0,0,0") -> (Pose) {translationMode=BMM_Additive, translationSpace=BCS_BoneSpace, boneToModify {boneName=pack}}
    Alpha
    Translation="0,0,0"
    Rotation="0,0,0"
   ) -> (Pose) {
    rotationMode=BMM_Additive
    rotationSpace=BCS_BoneSpace
    translationMode=BMM_Additive
    translationSpace=BCS_BoneSpace
    boneToModify {boneName=pack}
   }
   n9 = "Transform (Modify) Bone - Bone: base_lid"!(ComponentPose=@n8, bAlphaBoolEnabled:bool=True, Rotation="15,0,0") -> (Pose) {
    alphaInputType=Bool
    rotationMode=BMM_Additive
    rotationSpace=BCS_BoneSpace
    alphaBoolBlend {blendOption=Cubic, blendOutTime=0.5}
    boneToModify {boneName=base_lid}
    rotation {pitch=15}
   }
   n10 = "Transform (Modify) Bone - Bone: base_lid"!(ComponentPose=@n9, Alpha:float=1, Rotation="-33,0,0") -> (Pose) {
    rotationMode=BMM_Additive
    rotationSpace=BCS_BoneSpace
    alphaBoolBlend {blendOption=Cubic, blendOutTime=0.5}
    alphaScaleBiasClamp {bInterpResult=true, interpSpeedDecreasing=5, interpSpeedIncreasing=5}
    boneToModify {boneName=base_lid}
    rotation {pitch=-33}
   }
   Save cached pose 'preDynamicsCP'(Pose=Component To Local(ComponentPose=@n10)) {cacheName=preDynamicsCP}
  ]}
 EventGraph {nodes [
   n1 = Event BlueprintUpdateAnimation()
   > Set Delta Time X(Delta Time X=@n1.DeltaTimeX)
   > n2 = Sequence()
   > Set closeLid(closeLid:double=GetAlphaFromReferenceTransformDelta(SkeletalMeshComponent=GetOwnerSkeletalMeshComponent(), BoneName=spine_05, BaseName=root, InputRange=(X=40,Y=65), TwistAxis="1,0,0")) -> (Output_Get:double)
   n3 = Get bIsMoving2D()
   > n4 = Set isMoving2D float(isMoving2D float:double=FInterpTo(
     Current=Get isMoving2D float() -> (isMoving2D float:double)
     Target="To Float (Boolean)"(InBool=@n3)
     DeltaTime=Get Delta Time X()
     InterpSpeed=5
    )) -> (Output_Get:double)
   > Set Gravity Override(Gravity Override="Lerp (Vector)"!(A:Vector="0,0,-280", B:Vector="0,0,-230", Alpha:float=@n4) -> (ReturnValue:Vector))
   n5 = GetOwnerSkeletalMeshComponent()
   n6 = BreakRotator(InRot=BreakTransform(InTransform=GetDeltaTransformFromRefPose(self=@n5, BoneName=spine_05, BaseName=spine_01)).Rotation)
   Set backpack_translation(execute=@n2.then_1, backpack_translation=MakeVector(Y=MapRangeClamped(Value=@n6.Yaw, InRangeA=0, InRangeB=-30, OutRangeB=-4)))
   > Set backpack_rotate(backpack_rotate=MakeRotator(Roll=MapRangeClamped(Value=@n6.Yaw, InRangeA=0, InRangeB=-30, OutRangeB=-15)))
   n7 = Event OnCharacterCustomizationLoaded()
   n8 = Get Array Item with EFort Custom Part Type(Array:Array of Custom Character Part Object References=@n7.CustomParts) -> (Output:Custom Character Part Object Reference)
   > n9 = Branch(Condition="Equal (Enum)"(
     A:EFortCustomGender Enum=Get GenderPermitted!(self:Custom Character Part Object Reference=@n8) -> (GenderPermitted:EFortCustomGender Enum=Invalid)
     B:EFortCustomGender Enum=Male
    ))
   > Set offsetByGender(offsetByGender:Vector="0,0.8,5") -> (Output_Get:Vector)
   Set offsetByGender(execute=@n9.else, offsetByGender="0,1.6,6")
   n10 = Event BlueprintProcessAnimInputEvent()
   > Branch(Condition=IsNamed(self=@n10.AnimInputEvent, InEventName=aie.EnemyPlayerEliminated))
   > Set bEnemyPlayerEliminated(bEnemyPlayerEliminated:bool=true) -> (Output_Get:bool)
   > Delay(Duration=1)
   > Set bEnemyPlayerEliminated(bEnemyPlayerEliminated=false)
  ]}
]
variables [
 backpack_rotate {type=Rotator}
 backpack_translation {type=Vector}
 bEnemyPlayerEliminated {type=boolean}
 closeLid {type=number}
 Delta Time X {type=number}
 Gravity Override {type=Vector}
 isMoving2D float {type=number}
 offsetByGender {type=Vector}
]
